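# Material function for crowd materials that turns a scalar Crowd Response into two lerp alphas, staggered per instance.
# Per-instance random, slightly biased and rounded to tenths, sets each instance's threshold, and a steep saturated ramp above it blends that instance in.
# Lerp Alpha 1 follows response from 0 to 1, while Lerp Alpha 2 repeats the same staggered blend for response from 1 to 2.
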MaterialFunction /SparksMaps/Crowd/Materials/MF_Sparks_CrowdResponse_Blend_Offset
graph {nodes [
  n1 = FunctionInput() {inputName=Crowd Response, inputType=FunctionInput_Scalar}
  n2 = Add(A=PerInstanceRandom()) {constB=-0.01}
  n3 = Divide(A=Round(None=Multiply(A=@n2) {constB=10})) {constB=10}
  n4 = Subtract(A=Saturate(None=@n1), B=@n3)
  FunctionOutput(None=Saturate(None=Multiply(A=@n4) {constB=10})) {outputName=Lerp Alpha 1}
  n5 = Divide(A=@n1)
  n6 = Min(A=@n1, B=Multiply(A=Round(None=@n5)) {constB=2})
  n7 = Subtract(A=Saturate(None=Subtract(A=@n6)), B=@n3)
  FunctionOutput(None=Saturate(None=Multiply(A=@n7) {constB=10})) {outputName=Lerp Alpha 2, sortPriority=1}
 ]}
