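# Material function that animates Pilgrim crowd meshes from baked transform textures, outputting vertex offset (WPO), rotated normal and quaternion.
# Vertex color red picks a row in stacked position, rotation and optional scale bands, sampled across three animation textures blended by crowd response.
# Each vertex is taken relative to its sampled pivot, optionally scaled, rotated by quaternion, then translated.
# Static switches toggle WPO and scale, and a VertexInterpolator passes the quaternion to the pixel stage.
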MaterialFunction /SparksMaps/Crowd/Materials/MF_Pilgrim_WPOTransform_subuvs
graph {nodes [
  n1 = VertexColor()
  n2 = Multiply(
   A=FunctionInput() {inputName=vHeight, sortPriority=8}
   B=StaticSwitchParameter(True=Constant() {r=0.3333}, False=Constant() {r=0.5}) {parameterName=ConstainsScaleInfo}
  )
  n3 = Add(
   A=Add(A=Multiply(A=@n1.R, B=@n2) {constB=0.5}, B=FunctionInput() {inputName=vStart, sortPriority=7})
   B=Multiply(A=FunctionInput() {desc="height of a texel (1/yRez)", inputName=texelHeight, inputType=FunctionInput_Scalar, sortPriority=6}) {constB=0.5}
  ) {constB=-0.1}
  n4 = MaterialFunctionCall(X=NamedRerouteUsage(), Y=@n3) {materialFunction=&MakeFloat2}
  n5 = NamedRerouteUsage()
  n6 = NamedRerouteUsage()
  n7 = MaterialFunctionCall!(Crowd Response=NamedRerouteUsage()) {materialFunction=&/MF_Sparks_CrowdResponse_Blend_Offset}
  n8 = LinearInterpolate(
   A=LinearInterpolate(A=TextureSample(UVs=@n4.Result, Tex=@n5).RGB, B=TextureSample(UVs=@n4.Result, Tex=@n6).RGB, Alpha=@n7.Lerp Alpha 1)
   B=TextureSample(UVs=@n4.Result, Tex=NamedRerouteUsage()).RGB
   Alpha=@n7.Lerp Alpha 2
  )
  n9 = Add(A=@n3, B=@n2) {constB=0.5}
  n10 = MaterialFunctionCall(X=NamedRerouteUsage(), Y=Add(A=@n9, B=@n2)) {materialFunction=&MakeFloat2}
  n11 = TextureSample(UVs=@n10.Result, Tex=NamedRerouteUsage())
  n12 = TextureSample(UVs=@n10.Result, Tex=NamedRerouteUsage())
  n13 = MaterialFunctionCall(Crowd Response=NamedRerouteUsage()) {materialFunction=&MF_Sparks_CrowdResponse_Blend_Offset}
  n14 = NamedRerouteUsage()
  n15 = PreSkinnedPosition()
  n16 = MaterialFunctionCall(X=NamedRerouteUsage(), Y=@n3) {materialFunction=&MakeFloat2}
  n17 = NamedRerouteUsage()
  n18 = NamedRerouteUsage()
  n19 = MaterialFunctionCall(Crowd Response=NamedRerouteUsage()) {materialFunction=&MF_Sparks_CrowdResponse_Blend_Offset}
  n20 = LinearInterpolate(
   A=LinearInterpolate(A=TextureSample(UVs=@n16.Result, Tex=@n17).RGB, B=TextureSample(UVs=@n16.Result, Tex=@n18).RGB, Alpha=@n19.Lerp Alpha 1)
   B=TextureSample(UVs=@n16.Result, Tex=NamedRerouteUsage()).RGB
   Alpha=@n19.Lerp Alpha 2
  )
  n21 = Subtract(A=@n15, B=@n20) {constB=0}
  n22 = Multiply(
   A=@n21
   B=LinearInterpolate(
    A=LinearInterpolate(A=@n11.RGB, B=@n12.RGB, Alpha=@n13.Lerp Alpha 1)
    B=TextureSample(UVs=@n10.Result, Tex=@n14).RGB
    Alpha=@n13.Lerp Alpha 2
   )
  ) {constB=0.5}
  n23 = MaterialFunctionCall!(
   Vector=StaticSwitchParameter(True=StaticSwitchParameter(True=@n22, False=@n21) {group=Deformation, parameterName=ApplyScaleAnimation, sortPriority=10}, False=@n21) {group=BakedTransforms, parameterName=ConstainsScaleInfo, sortPriority=25}
   Quat=NamedRerouteUsage()
   Invert Quaternion
  ) {materialFunction=/Engine/Functions/Engine_MaterialFunctions03/Math/MultiplyVectorWithQuaternion}
  n24 = Subtract(A=Add(A=Add(A=@n8, B=@n23.Result) {constB=0}, B=@n20) {constB=0}, B=@n15)
  FunctionOutput(None=StaticSwitchParameter(True=Transform(None=@n24) {transformSourceType=TRANSFORMSOURCE_Local}, False=Constant3Vector()) {defaultValue=true, group=Deformation, parameterName=ApplyWPO, sortPriority=5}) {outputName=WPO}
  n25 = Transform(None=FunctionInput(Preview=VertexNormalWS()) {bUsePreviewValueAsDefault=true, inputName=WorldNormal, sortPriority=50}) {transformSourceType=TRANSFORMSOURCE_World, transformType=TRANSFORM_Instance}
  n26 = NamedRerouteUsage()
  n27 = VertexInterpolator(VS=@n26)
  FunctionOutput(None=Transform(None=MaterialFunctionCall(Vector=@n25, Quat=@n27.PS) {materialFunction=&MultiplyVectorWithQuaternion}.Result) {transformSourceType=TRANSFORMSOURCE_Instance}) {outputName=Normal, sortPriority=-2}
  FunctionOutput(None=@n27.PS) {outputName=Quaternion}
  FunctionOutput(None=@n26) {outputName=QuaternionVS}
  NamedRerouteDeclaration(None=FunctionInput() {inputName=AnimationState, inputType=FunctionInput_Scalar, sortPriority=3}) {name=AnimationState, nodeColor=<rgba 0.125 1 0 1>}
  NamedRerouteDeclaration(None=FunctionInput() {inputName=AnimationTexture, inputType=FunctionInput_Texture2D}) {name=AnimationTexture, nodeColor=<rgba 0 1 0.2969 1>}
  NamedRerouteDeclaration(None=FunctionInput() {inputName=AnimationTexture2, inputType=FunctionInput_Texture2D, sortPriority=1}) {name=AnimationTexture2, nodeColor=<rgba 0 1 1 1>}
  NamedRerouteDeclaration(None=FunctionInput() {inputName=AnimationTexture3, inputType=FunctionInput_Texture2D, sortPriority=2}) {name=AnimationTexture3, nodeColor=<rgba 0.0078 0 1 1>}
  NamedRerouteDeclaration(None=FunctionInput() {desc=u value to sample, inputName=u, inputType=FunctionInput_Scalar, sortPriority=4}) {name=u, nodeColor=<rgba 1 0.0703 0 1>}
  NamedRerouteDeclaration(None=FunctionInput() {desc=u value of bind pose, inputName=bind_u, inputType=FunctionInput_Scalar, sortPriority=5}) {name=bind_u, nodeColor=<rgba 0.9688 1 0 1>}
  n28 = MaterialFunctionCall(X=NamedRerouteUsage(), Y=@n9) {materialFunction=&MakeFloat2}
  n29 = TextureSample(UVs=@n28.Result, Tex=NamedRerouteUsage())
  n30 = TextureSample(UVs=@n28.Result, Tex=NamedRerouteUsage())
  n31 = MaterialFunctionCall(Crowd Response=NamedRerouteUsage()) {materialFunction=&MF_Sparks_CrowdResponse_Blend_Offset}
  n32 = NamedRerouteUsage()
  NamedRerouteDeclaration(None=LinearInterpolate(
    A=LinearInterpolate(A=@n29.RGBA, B=@n30.RGBA, Alpha=@n31.Lerp Alpha 1)
    B=TextureSample(UVs=@n28.Result, Tex=@n32).RGBA
    Alpha=@n31.Lerp Alpha 2
   )) {name=Rotation Lookup, nodeColor=<rgba 1 0.0703 0 1>}
 ]}
