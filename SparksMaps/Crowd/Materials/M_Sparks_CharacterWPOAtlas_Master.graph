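# Master material for Sparks crowd characters on instanced and Niagara meshes, built around texture-baked World Position Offset animation.
# A per-primitive CrowdResponse value selects among three animation textures through a blend-offset function.
# Playback time is offset by world position and per-instance random to desynchronise characters, then feeds a transform function whose normals blend with a normal map.
# Static switches add masked Fresnel skin tinting and a skin-weight preview.
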Material /SparksMaps/Crowd/Materials/M_Sparks_CharacterWPOAtlas_Master
graph {
 properties {
  bTangentSpaceNormal=false
  bUsedWithInstancedStaticMeshes=true
  bUsedWithNiagaraMeshParticles=true
  pixelDepthOffsetMode=PDOM_Legacy
  usedShadingModels=Default Lit
 }
 nodes [
  CollectionParameter() {collection=/SparksEffects/Effects/MPC_Sparks, desc="0 = Animation 1\r\n1 = Animation 2\r\n2 = Animation 3", parameterName=CrowdResponse}
  Constant() {r=0.8}
  n1 = DynamicParameter() {defaultValue=<rgba 0 0 0 0>}
  MaterialFunctionCall(X=@n1.Param1, Y=@n1.Param1, Z=@n1.Param1, A=@n1.Param1) {materialFunction=&MakeFloat4}
  MaterialFunctionCall(
   BaseNormal=MaterialFunctionCall!(
    AnimationTexture=NamedRerouteUsage()
    AnimationTexture2=NamedRerouteUsage()
    AnimationTexture3=NamedRerouteUsage()
    AnimationState=NamedRerouteUsage()
    u=NamedRerouteUsage()
    bind_u=NamedRerouteUsage()
    texelHeight=NamedRerouteUsage()
    vStart=Constant()
    vHeight=Constant() {r=1}
    WorldNormal
   ) {materialFunction=&/MF_Pilgrim_WPOTransform_subuvs}.Normal
   AdditionalNormal=TextureSampleParameter2D() {parameterName=Normals, samplerType=SAMPLERTYPE_Normal, texture=&FlatNormalBlue}.RGB
  ) {materialFunction=&BlendAngleCorrectedNormals}
  n2 = VertexColor()
  n3 = Divide(A=Divide(A=ScalarParameter() {parameterName=jointIndex}, B=Constant() {r=26}))
  n4 = VertexColor()
  n5 = Constant3Vector() {constant=<rgba 1 0 0 0>}
  NamedRerouteDeclaration(None=StaticSwitchParameter(
    True=SphereMask(A=@n2, B=@n3) {attenuationRadius=0.5}
    False=MaterialFunctionCall(Hue Shift Percentage=@n4.R, Texture=@n5) {materialFunction=&HueShift}.Result
   ) {parameterName=pickWeight}) {name=PreviewSkinWeights, nodeColor=<rgba 0 0.9062 1 1>}
  n6 = TextureObjectParameter() {
   parameterName=Animation Texture
   samplerType=SAMPLERTYPE_LinearColor
   texture=/SparksMaps/Crowd/Textures/SoldierMannequin_Male_GalileoShow_Gesture
  }
  n7 = TextureProperty(Tex=NamedRerouteDeclaration(None=@n6) {name=AnimationTexture, nodeColor=<rgba 0 1 0.1328 1>})
  n8 = TextureObjectParameter() {parameterName=Animation Texture 2, samplerType=SAMPLERTYPE_LinearColor, texture=/SparksMaps/Crowd/Textures/T_Sparks_GalileoShow_Cheer_Male}
  n9 = TextureProperty(Tex=NamedRerouteDeclaration(None=@n8) {name=AnimationTexture2, nodeColor=<rgba 0.5703 1 0 1>})
  n10 = MaterialFunctionCall!(Crowd Response=NamedRerouteDeclaration(None=ScalarParameter() {bUseCustomPrimitiveData=true, parameterName=CrowdResponse, primitiveDataIndex=5}) {name=CrowdResponse, nodeColor=<rgba 0.7109 0 1 1>}) {materialFunction=&/MF_Sparks_CrowdResponse_Blend_Offset}
  n11 = Round(None=@n10.Lerp Alpha 1)
  n12 = TextureObjectParameter() {
   parameterName=Animation Texture 3
   samplerType=SAMPLERTYPE_LinearColor
   texture=/SparksMaps/Crowd/Textures/T_Sparks_PapayaComms_Cheer_CMM_Male
  }
  n13 = NamedRerouteDeclaration(None=@n12) {name=AnimationTexture3, nodeColor=<rgba 0 0.3438 1 1>}
  n14 = MaterialFunctionCall(Float2=LinearInterpolate(A=LinearInterpolate(A=@n7, B=@n9, Alpha=@n11), B=TextureProperty(Tex=@n13), Alpha=Round(None=@n10.Lerp Alpha 2))) {materialFunction=&BreakOutFloat2Components}
  n15 = Divide(B=NamedRerouteDeclaration(None=@n14.R) {name=textureWidth, nodeColor=<rgba 0.0078 0 1 1>})
  NamedRerouteDeclaration(None=Multiply(A=NamedRerouteDeclaration(None=@n15) {name=texelWidth, nodeColor=<rgba 0 1 0.2266 1>}) {constB=0.5}) {name=halfTexelWidth, nodeColor=<rgba 0.0312 0 1 1>}
  n16 = ObjectPositionWS()
  n17 = Add(A=ComponentMask(None=@n16) {r=true}, B=ComponentMask(None=@n16) {g=true})
  n18 = Add(A=Add(A=Time(), B=Divide(A=@n17) {constB=100}), B=PerInstanceRandom())
  n19 = Divide(B=NamedRerouteUsage()) {constA=30}
  n20 = Frac(None=Add(A=Multiply(A=@n18, B=@n19)) {constB=0})
  n21 = NamedRerouteUsage()
  NamedRerouteDeclaration(None=Add(A=Multiply(A=@n20, B=@n21), B=NamedRerouteUsage())) {name=u, nodeColor=<rgba 0 1 0.9766 1>}
  NamedRerouteDeclaration(None=Add(A=NamedRerouteUsage(), B=NamedRerouteUsage())) {name=uStart, nodeColor=<rgba 1 0.3516 0 1>}
  n22 = NamedRerouteUsage()
  NamedRerouteDeclaration(None=Subtract(A=Subtract(B=@n22), B=@n22)) {name=uWidth, nodeColor=<rgba 0.9922 1 0 1>}
  NamedRerouteDeclaration(None=Divide(B=NamedRerouteDeclaration(None=@n14.G) {name=textureHeight, nodeColor=<rgba 0.5469 0 1 1>})) {name=texelHeight, nodeColor=<rgba 0 1 0.2266 1>}
  n23 = TextureSampleParameter2D() {parameterName=BaseColor, texture=&T_BlacksRGB}
  n24 = Constant() {desc=Global Skin Boost EX, r=0.75}
  n25 = VectorParameter() {group=SkinColor, parameterName=Skin Boost Color And Exponent, defaultValue=<rgba 0.87 0.6688 0.6948 2>}
  n26 = Fresnel(ExponentIn=@n25.A) {exponent=2}
  n27 = Multiply(A=@n23.RGB, B=@n25.RGB)
  n28 = Add(A=Multiply(A=OneMinus(None=@n26), B=@n27), B=@n27)
  n29 = Saturate(None=LinearInterpolate(A=@n23.RGB, B=Multiply(A=@n24, B=@n28), Alpha=Constant() {desc=Human Skin, r=1}))
  n30 = TextureSampleParameter2D() {parameterName=SkinMask, texture=/Engine/EngineResources/Black_Low.Black}
  NamedRerouteDeclaration(None=StaticSwitchParameter(True=LinearInterpolate(A=@n23.RGB, B=@n29, Alpha=@n30.B), False=@n23.RGB) {parameterName=ChangeSkinColor}) {name=Base Color, nodeColor=<rgba 0 0.4375 1 1>}
  NamedRerouteUsage()
  NamedRerouteUsage()
 ]
}
