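# Material function for Sparks crowd meshes that yaws each one about its pivot to face a LookAtLocation input.
# The angle comes from Arctangent2 of the XY direction between object position and target, converted to turns with a quarter-turn correction.
# A pseudo-random jitter from the fraction of summed pivot X and Y, scaled by Offset Amount, varies the facing.
# RotateVertexAboutAxis applies the Z rotation to position plus incoming WPO and to the normal, outputting both.
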MaterialFunction /SparksMaps/Crowd/Materials/MF_Sparks_Crowd_Rotate_LookAt
graph {nodes [
  n1 = ObjectPositionWS()
  n2 = Multiply(A=FunctionInput() {inputName=LookAtLocation}, B=Constant3Vector() {constant=<rgba 1 1 0 0>})
  n3 = MaterialFunctionCall(Color=Normalize(VectorInput=Subtract(A=@n1, B=@n2))) {materialFunction=&SplitComponents}
  n4 = Divide(A=Divide(A=Arctangent2!(Y=@n3.G, X=@n3.G), B=Constant() {r=3.142}))
  n5 = MaterialFunctionCall() {materialFunction=&ObjectPivotPoint}
  n6 = Add(A=ComponentMask(None=@n5.Object Pivot Location) {r=true}, B=ComponentMask(None=@n5.Object Pivot Location) {g=true})
  n7 = Divide(A=Subtract(A=Frac(None=@n6)) {constB=0.5}) {constB=10}
  n8 = FunctionInput() {inputName=Offset Amount}
  n9 = MaterialFunctionCall!(
   RotationAngle=Add(A=Add(B=@n4) {constA=0.25}, B=Multiply(A=@n7, B=@n8))
   NormalizedRotationAxis=Constant3Vector() {constant=<rgba 0 0 1 0>}
   PivotPoint=@n5.Object Pivot Location
   Position=Add(A=FunctionInput() {inputName=WPO}, B=WorldPosition().XYZ)
   Normal=FunctionInput() {inputName=Normal}
   TangentU
   TangentV
  ) {materialFunction=/Game/Characters/Player/Male/Medium/Bodies/M_MED_Inferno/FX/Mat/RotateVertexAboutAxis}
  FunctionOutput(None=@n9.WPO) {outputName=WPO}
  FunctionOutput(None=@n9.Normal) {outputName=Normal}
 ]}
