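# Master material for Sparks stands crowd characters, animating instanced, Nanite or Niagara meshes through texture-driven world position offset.
# The MPC_Sparks CrowdResponse value picks among three animation textures, with playback staggered by time, object position and per-instance random.
# Sampled offsets and normals pass through a look-at rotation driven by custom primitive data.
# Static switches cover WPO, skin recoloring, stage-colored glow stick emissive, team colors and skin-weight preview.
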Material /SparksMaps/Crowd/Materials/M_Sparks_CharacterWPOAtlas_Stands_Master
graph {
 properties {
  bTangentSpaceNormal=false
  bUsedWithInstancedStaticMeshes=true
  bUsedWithNanite=true
  bUsedWithNiagaraMeshParticles=true
  pixelDepthOffsetMode=PDOM_Legacy
  usedShadingModels=Default Lit
 }
 nodes [
  Constant() {r=0.8}
  MaterialFunctionCall!(BaseColor=NamedRerouteUsage(), Amount=Constant() {r=0.2}) {materialFunction=/Engine/Functions/Engine_MaterialFunctions01/SpeedTree/SpeedTreeColorVariation}
  Multiply(A=NamedRerouteUsage(), B=NamedRerouteUsage())
  n1 = VertexColor()
  n2 = Divide(A=Divide(A=ScalarParameter() {parameterName=jointIndex}, B=Constant() {r=26}))
  n3 = VertexColor()
  n4 = Constant3Vector() {constant=<rgba 1 0 0 0>}
  NamedRerouteDeclaration(None=StaticSwitchParameter(
    True=SphereMask(A=@n1, B=@n2) {attenuationRadius=0.5}
    False=MaterialFunctionCall(Hue Shift Percentage=@n3.R, Texture=@n4) {materialFunction=&HueShift}.Result
   ) {parameterName=pickWeight}) {name=PreviewSkinWeights, nodeColor=<rgba 0 0.9062 1 1>}
  n5 = TextureSampleParameter2D() {
   parameterName=AudioReactivityMask
   samplerType=SAMPLERTYPE_Masks
   texture=/Game/Characters/Player/Male/Medium/Bodies/M_MED_Stands/Textures/T_M_MED_Stands_Body_Mask
  }
  NamedRerouteDeclaration(None=@n5.R) {name=Emissive Mask, nodeColor=<rgba 1 0 0.3984 1>}
  n6 = TextureObjectParameter() {
   parameterName=Animation Texture
   samplerType=SAMPLERTYPE_LinearColor
   texture=/SparksMaps/Crowd/Textures/SoldierMannequin_Male_GalileoShow_Gesture
  }
  n7 = TextureProperty(Tex=NamedRerouteDeclaration(None=@n6) {name=AnimationTexture, nodeColor=<rgba 0 1 0.1328 1>})
  n8 = TextureObjectParameter() {parameterName=Animation Texture 2, samplerType=SAMPLERTYPE_LinearColor, texture=/SparksMaps/Crowd/Textures/T_Sparks_GalileoShow_Cheer_Male}
  n9 = TextureProperty(Tex=NamedRerouteDeclaration(None=@n8) {name=AnimationTexture2, nodeColor=<rgba 0.5703 1 0 1>})
  n10 = CollectionParameter() {collection=/SparksEffects/Effects/MPC_Sparks, desc="0 = Animation 1\r\n1 = Animation 2\r\n2 = Animation 3", parameterName=CrowdResponse}
  n11 = MaterialFunctionCall!(Crowd Response=NamedRerouteDeclaration(None=@n10) {name=CrowdResponse, nodeColor=<rgba 0.0078 1 0 1>}) {materialFunction=&/MF_Sparks_CrowdResponse_Blend_Offset}
  n12 = Round(None=@n11.Lerp Alpha 1)
  n13 = TextureObjectParameter() {
   parameterName=Animation Texture 3
   samplerType=SAMPLERTYPE_LinearColor
   texture=/SparksMaps/Crowd/Textures/T_Sparks_PapayaComms_Cheer_CMM_Male
  }
  n14 = NamedRerouteDeclaration(None=@n13) {name=AnimationTexture3, nodeColor=<rgba 0 0.3437 1 1>}
  n15 = MaterialFunctionCall(Float2=LinearInterpolate(A=LinearInterpolate(A=@n7, B=@n9, Alpha=@n12), B=TextureProperty(Tex=@n14), Alpha=Round(None=@n11.Lerp Alpha 2))) {materialFunction=&BreakOutFloat2Components}
  n16 = Divide(B=NamedRerouteDeclaration(None=@n15.R) {name=textureWidth, nodeColor=<rgba 0.0078 0 1 1>})
  NamedRerouteDeclaration(None=Multiply(A=NamedRerouteDeclaration(None=@n16) {name=texelWidth, nodeColor=<rgba 0 1 0.2266 1>}) {constB=0.5}) {name=halfTexelWidth, nodeColor=<rgba 0.0312 0 1 1>}
  n17 = ObjectPositionWS()
  n18 = Add(A=ComponentMask(None=@n17) {r=true}, B=ComponentMask(None=@n17) {g=true})
  n19 = Add(A=Add(A=Time(), B=Divide(A=@n18) {constB=100}), B=PerInstanceRandom())
  n20 = Divide(B=NamedRerouteUsage()) {constA=30}
  n21 = Frac(None=Add(A=Multiply(A=@n19, B=@n20)) {constB=0})
  n22 = NamedRerouteUsage()
  NamedRerouteDeclaration(None=Add(A=Multiply(A=@n21, B=@n22), B=NamedRerouteUsage())) {name=u, nodeColor=<rgba 0 1 0.9766 1>}
  NamedRerouteDeclaration(None=Add(A=NamedRerouteUsage(), B=NamedRerouteUsage())) {name=uStart, nodeColor=<rgba 1 0.3516 0 1>}
  n23 = NamedRerouteUsage()
  NamedRerouteDeclaration(None=Subtract(A=Subtract(B=@n23), B=@n23)) {name=uWidth, nodeColor=<rgba 0.9922 1 0 1>}
  NamedRerouteDeclaration(None=Divide(B=NamedRerouteDeclaration(None=@n15.G) {name=textureHeight, nodeColor=<rgba 0.5469 0 1 1>})) {name=texelHeight, nodeColor=<rgba 0 1 0.2266 1>}
  n24 = Constant3Vector() {constant=<rgba 0.25 0.25 0.25 0>}
  n25 = ScalarParameter() {parameterName=Hostile/Friendly, sliderMax=1, sliderMin=-1}
  n26 = LinearInterpolate(A=@n24, B=Constant3Vector() {constant=<rgba 1 0 0 0>}, Alpha=Saturate(None=Multiply(A=@n25) {constB=-1}))
  n27 = LinearInterpolate(A=@n24, B=Constant3Vector() {constant=<rgba 0 1 0 1>}, Alpha=Saturate(None=@n25))
  n28 = Saturate(None=Ceil(None=@n25))
  NamedRerouteDeclaration(None=StaticSwitchParameter(
    True=LinearInterpolate(A=@n26, B=@n27, Alpha=@n28)
    False=VectorParameter() {parameterName=Audio Primary Color, defaultValue=<rgba 0.9271 0.0922 0.5072 1>}.RGB
   ) {parameterName=TeamBased?}) {name=Emissive, nodeColor=<rgba 1 0 0.0937 1>}
  n29 = TextureSampleParameter2D() {parameterName=BaseColor, texture=&T_BlacksRGB}
  n30 = Constant() {desc=Global Skin Boost EX, r=0.75}
  n31 = VectorParameter() {group=SkinColor, parameterName=Skin Boost Color And Exponent, defaultValue=<rgba 0.87 0.6688 0.6948 2>}
  n32 = Fresnel(ExponentIn=@n31.A) {exponent=2}
  n33 = Multiply(A=@n29.RGB, B=@n31.RGB)
  n34 = Add(A=Multiply(A=OneMinus(None=@n32), B=@n33), B=@n33)
  n35 = Saturate(None=LinearInterpolate(A=@n29.RGB, B=Multiply(A=@n30, B=@n34), Alpha=Constant() {desc=Human Skin, r=1}))
  n36 = TextureSampleParameter2D() {parameterName=SkinMask, texture=/Engine/EngineResources/Black_Low.Black}
  NamedRerouteDeclaration(None=StaticSwitchParameter(True=LinearInterpolate(A=@n29.RGB, B=@n35, Alpha=@n36.B), False=@n29.RGB) {parameterName=ChangeSkinColor}) {name=Base Color, nodeColor=<rgba 0 0.4375 1 1>}
  n37 = Multiply(
   A=OneMinus(None=VertexColor().A)
   B=StaticSwitchParameter(
    True=MaterialFunctionCall!(CustomUPosition) {materialFunction=/SparksStageShow/Systems/StageManager/FXManager/MaterialFunctions/MF_MojoStageColors}.Primary
    False=VectorParameter() {group=Emissive, parameterName=Emissive Color, defaultValue=<rgba 0 1 0.3553 1>}.RGB
   ) {defaultValue=true, parameterName=Use Stage Colors for Emissive}
  )
  n38 = ScalarParameter() {defaultValue=100, group=Emissive, parameterName=Emissive Intensity}
  NamedRerouteDeclaration(None=StaticSwitchParameter(True=Multiply(A=@n37, B=@n38) {constB=100}, False=Constant()) {parameterName=UseEmissive}) {name=Glow Stick, nodeColor=<rgba 1 0 0.1875 1>}
  NamedRerouteUsage()
  NamedRerouteUsage()
  n39 = WorldPosition()
  n40 = VectorParameter() {parameterName=Stage Position, defaultValue=<rgba 0 0 0 1>}
  n41 = Add(
   A=Multiply(A=Time(), B=ScalarParameter() {defaultValue=-1, parameterName=Wave Speed}) {constB=-1}
   B=Multiply(A=Distance(A=@n39.XYZ, B=@n40.RGB), B=ScalarParameter() {defaultValue=0.001, parameterName=Wave Thickness})
  )
  Saturate(None=SmoothStep(Value=Sine(None=@n41)) {constMin=-0.5})
  ScalarParameter() {bUseCustomPrimitiveData=true, defaultValue=1, parameterName=CrowdResponse, primitiveDataIndex=1}
  n42 = Transform(None=TextureSampleParameter2D() {parameterName=Normals, samplerType=SAMPLERTYPE_Normal, texture=&FlatNormalBlue}.RGB)
  n43 = MaterialFunctionCall!(
   AnimationTexture=NamedRerouteUsage()
   AnimationTexture2=NamedRerouteUsage()
   AnimationTexture3=NamedRerouteUsage()
   AnimationState=NamedRerouteUsage()
   u=NamedRerouteUsage()
   bind_u=NamedRerouteUsage()
   texelHeight=NamedRerouteUsage()
   vStart=Constant()
   vHeight=Constant() {r=1}
   WorldNormal=@n42
  ) {materialFunction=&/MF_Pilgrim_WPOTransform_subuvs}
  n44 = MaterialFunctionCall!(
   Offset Amount=ScalarParameter() {bUseCustomPrimitiveData=true, parameterName=OffsetAmount, primitiveDataIndex=4}
   Normal=@n43.Normal
   LookAtLocation=VectorParameter() {bUseCustomPrimitiveData=true, parameterName=LookAtLocation, defaultValue=<rgba 0 0 0 1>}.RGB
   WPO=@n43.Normal
  ) {materialFunction=&/MF_Sparks_Crowd_Rotate_LookAt}
  StaticSwitchParameter(True=Add(A=@n43.WPO, B=@n44.WPO), False=Constant3Vector()) {defaultValue=true, parameterName=UseWPO}
  StaticSwitchParameter(True=@n44.Normal, False=@n42) {defaultValue=true, parameterName=UseWPO}
 ]
}
