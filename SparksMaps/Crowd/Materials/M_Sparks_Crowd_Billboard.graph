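# Unlit masked material for instanced crowd billboards in the Sparks map, showing animated spectators that turn to face the camera.
# A 6x6 flipbook plays with a per-instance random time offset.
# Its darkened, tinted result passes through SpeedTree color variation, then is multiplied by a vertical shadow gradient and skylight color.
# Vertex math rebuilds a basis from the horizontal camera-to-object direction and world up, rotating each card around its vertical axis.
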Material /SparksMaps/Crowd/Materials/M_Sparks_Crowd_Billboard
graph {
 properties {
  bAutomaticallySetUsageInEditor=false
  bUsedWithInstancedStaticMeshes=true
  blendMode=BLEND_Masked
  shadingModel=MSM_Unlit
  usedShadingModels=Unlit
 }
 nodes [
  n1 = TextureCoordinate()
  n2 = ComponentMask(None=@n1) {g=true}
  AppendVector(A=ComponentMask(None=@n1) {r=true}, B=Add(A=Multiply(A=@n2) {constB=0.5}) {constB=0.5})
  n3 = Time()
  n4 = Multiply(A=Frac(None=PerInstanceRandom())) {constB=100}
  n5 = MaterialFunctionCall(
   "Animation  Phase (0-1)"=Divide(A=Add(A=@n3, B=@n4), B=ScalarParameter() {defaultValue=6, parameterName=Speed})
   Number of Rows=Constant() {r=6}
   Number of Columns=Constant() {r=6}
   Texture=TextureObject() {texture=/SparksMaps/Crowd/Textures/T_Sparks_CrowdFlipbook}
   UVs=@n1
  ) {materialFunction=&FlipBook}
  Multiply(A=@n5.Result) {constB=5}
  n6 = OneMinus(None=ComponentMask(None=TextureCoordinate()) {g=true})
  n7 = ScalarParameter() {defaultValue=7, parameterName=Shadow Gradient Power}
  n8 = Multiply(
   A=Subtract(A=@n5.Result, B=ScalarParameter() {defaultValue=0.4, parameterName=Darken, sliderMax=1}) {constB=0.4}
   B=ComponentMask(None=VectorParameter() {parameterName=Color, defaultValue=<rgba 1 1 1 0>}.RGB) {b=true, g=true, r=true}
  )
  n9 = ScalarParameter() {defaultValue=0.6, parameterName=Color Variation Amount}
  Multiply(
   A=Multiply(
    A=Power(Base=@n6, Exp=@n7) {constExponent=7}
    B=MaterialFunctionCall!(BaseColor=@n8, Amount=@n9) {materialFunction=/Engine/Functions/Engine_MaterialFunctions01/SpeedTree/SpeedTreeColorVariation}.Result
   )
   B=MaterialFunctionCall() {materialFunction=/Game/Packages/Fortress_SharedMaterials/Base_Material_Functions/SkylightColorAndIntensity.SkyLightColorAndIntensity}.Color
  )
  n10 = Subtract(A=WorldPosition().XYZ, B=ObjectPositionWS())
  n11 = ComponentMask(None=ObjectPositionWS()) {g=true, r=true}
  n12 = ComponentMask(None=CameraPositionWS()) {g=true, r=true}
  n13 = AppendVector(A=Normalize(VectorInput=Subtract(A=@n11, B=@n12)), B=Constant())
  n14 = Constant3Vector() {constant=<rgba 0 0 1 0>}
  n15 = CrossProduct(A=@n13, B=@n14)
  Subtract(
   A=Add(
    A=MaterialFunctionCall!(VectorToTransform=@n10, BasisX=@n13, BasisY=@n15, BasisZ=@n14) {materialFunction=/Engine/Functions/Engine_MaterialFunctions02/Math/InverseTransformMatrix}.Transformed Vector 3
    B=ObjectPositionWS()
   )
   B=WorldPosition().XYZ
  )
  n16 = Subtract(
   A=ComponentMask(None=ObjectPositionWS() {originType=CameraRelative}) {g=true, r=true}
   B=ComponentMask(None=CameraPositionWS()) {g=true, r=true}
  )
  Transform(None=AppendVector(A=Normalize(VectorInput=@n16), B=Constant())) {transformSourceType=TRANSFORMSOURCE_World, transformType=TRANSFORM_Tangent}
 ]
}
