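; Crowd volume actor for Sparks music maps that builds an audience and aims its gaze through custom primitive data on the spawner's child meshes.
; In editor, construction configures a PCG spline crowd or sizes preview boxes for the legacy box crowd, which SparksCrowdSpawner spawns.
; Spline crowds snap their gaze at BeginPlay to the player or stage, depending on spline proximity.
; Quickplay crowds switch to a secondary target during song or set results.
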(event EventBeginPlay
  ;[1 1 1] Listen for Song Results and set Look At Target
  (if (Variables|Crowd|GetIsQuickplayCrowd)
    (bind (_asynctaskproxy _playspace) (Playspace|Discoverable|WaitForDiscoverablePlayspaceOfTypeForActor "/Script/SparksMusicPlayspaceRuntime.SparksMusicPlayspace" self)
      (:OnFoundPlayspace
        (bind _assparks_music_playspace (Utilities|Casting|CastToSparksMusicPlayspace _playspace)
          (:then
            (Variables|Crowd|SetPlaySpace _assparks_music_playspace)
            (bind (_asynctaskproxy_2 _payload) (GameplayEventRouter|StartListeningToEvent (Class|SparksMusicPlayspace|GetOwnedEventRouter (Variables|Crowd|GetPlaySpace)) "/Script/PilgrimQPSharedRuntime.PilgrimEvent_QuickplayState_Began")
              (:then
                _)
              (:OnEventReceived
                (if (or (Utilities|Enum|Equal(Enum) (Utilities|Struct|BreakPilgrimEventQuickplayStateBegan _payload) "SongResults") (Utilities|Enum|Equal(Enum) (Utilities|Struct|BreakPilgrimEventQuickplayStateBegan _payload) "SetResults"))
                  (CallFunction|ChangeLookatTarget (Variables|Default|GetLookAtArea_Secondary))
                  (else
                    (CallFunction|ChangeLookatTarget (Variables|Default|GetLookAtArea)))))))
          (:CastFailed
            (Fortnite|FortShippingLog self (Utilities|String|Append "WaitForPlayspace: Given playspace " (Utilities|GetDisplayName _playspace) " which doesn't derive from SparksMusicPlayspace.") true)))))
    (else
      _))
  ;[1 1 1] Set Look At Offset Amount in Material
  (for _array_element (Components|GetChildrenComponents (Variables|Default|GetSparksCrowdSpawner) false)
    (bind _asmesh_component (Utilities|Casting|CastToMeshComponent _array_element))
    (Rendering|Material|SetCustomPrimitiveDataFloat _asmesh_component 4 (Variables|Look|GetLookAtOffsetAmount)))
  (if (Variables|Crowd|GetUseSplineCrowd)
    (CallFunction|SetupSplineVolume)
    (CallFunction|LookatTarget)))

(event Custom|LookAtTarget
  ;[1 1 1] Set Crowd Look At Location Parameter in Material (instant look at)
  (bind (_constantratelookatlocation _instantlookatlocation) (CallFunction|SetLookatPlayerStage 1.5e+03 1e+03 (Variables|Default|GetLookAtArea)))
  (Variables|Look|SetLookAtLocation_ConstantRate _instantlookatlocation)
  (for _array_element (Components|GetChildrenComponents (Variables|Default|GetSparksCrowdSpawner) false)
    (bind _asmesh_component (Utilities|Casting|CastToMeshComponent _array_element))
    (Rendering|Material|SetCustomPrimitiveDataVector3 _asmesh_component 0 _instantlookatlocation)))

(fn ChangeLookAtTarget
  (params (Target :type Billboard%20Component%20Object%20Reference))
  (for _array_element (Components|GetChildrenComponents (Variables|Default|GetSparksCrowdSpawner) false)
    (bind _asmesh_component (Utilities|Casting|CastToMeshComponent _array_element))
    (Rendering|Material|SetCustomPrimitiveDataVector3 _asmesh_component 0 (Transformation|GetWorldLocation Target))))

(fn GetLookAtLocation
  (return (+ (Transformation|GetWorldLocation (Variables|Look|GetEDITOR_LookAtArea)) (Math|Transform|TransformDirection (Transformation|GetActorTransform) (Variables|LegacyCrowd|GetLookAtOffset))) (Math|Transform|InverseTransformLocation (Transformation|GetActorTransform) (+ (Transformation|GetWorldLocation (Variables|Look|GetEDITOR_LookAtArea)) (Math|Transform|TransformDirection (Transformation|GetActorTransform) (Variables|LegacyCrowd|GetLookAtOffset))))))

(fn Respawn
  (LegacyCrowd|SpawnCrowd true))

(fn SetLookAtPlayerStage
  (params (ConstantRotationSpeed :type Float%20%28double-precision%29 :default 1e+03) (Inclusive%20Radius :type Float%20%28double-precision%29 :default 1e+03) (LookAtArea :type Billboard%20Component%20Object%20Reference))
  (bind _returnvalue (Transformation|GetWorldLocation LookAtArea))
  (bind _returnvalue_2 (Game|GetPlayerCharacter 0))
  (bind _returnvalue_3 (Transformation|GetActorLocation _returnvalue_2))
  (bind _spline (Variables|Default|GetSpline))
  (bind _returnvalue_4 (Spline|FindLocationClosesttoWorldLocation _spline _returnvalue_3 "World"))
  (bind _returnvalue_5 (- _returnvalue_3 _returnvalue_4))
  (bind _returnvalue_6 (Math|Vector|VectorLength _returnvalue_5))
  (bind _returnvalue_7 (Game|GetPlayerCharacter 0))
  (bind _returnvalue_8 (Transformation|GetWorldLocation LookAtArea))
  (bind _returnvalue_9 (Transformation|GetActorLocation _returnvalue_7))
  (bind _returnvalue_10 (<= _returnvalue_6 Inclusive%20Radius))
  (bind _returnvalue_11 (select _returnvalue_10 _returnvalue_9 _returnvalue_8))
  (Utilities|IsValid LookAtArea
    (:Is%20Valid
      (Utilities|IsValid (Game|GetPlayerCharacter 0)
        (:Is%20Valid
          ;[1 1 1] lerp to stage or player locaiton depending on distance
          (bind _output_get (Variables|Look|SetLookAtLocation_ConstantRate (Math|Interpolation|VInterptoConstant (Variables|Look|GetLookAtLocation_ConstantRate) _returnvalue_11 (Utilities|Time|GetWorldDeltaSeconds) ConstantRotationSpeed)))
          (return _output_get _returnvalue_11))
        (:Is%20Not%20Valid
          ;[1 1 1] lerp to stage
          (bind _output_get_2 (Variables|Look|SetLookAtLocation_ConstantRate (Math|Interpolation|VInterptoConstant (Variables|Look|GetLookAtLocation_ConstantRate) _returnvalue (Utilities|Time|GetWorldDeltaSeconds) ConstantRotationSpeed)))
          (return _output_get_2 _returnvalue))))))

(fn SetupSplineVolume
  (bind _usesplinecrowd (Variables|Crowd|GetUseSplineCrowd))
  (bind _hism_crowd_05 (Variables|Default|GetHISM_Crowd_05))
  (if (Editor|InEditor)
    (Variables|/Script/PCG.PCGComponent|SetbActivated _usesplinecrowd (Variables|Default|GetPCG))
    (if _usesplinecrowd
      (Components|InstancedStaticMesh|ClearInstances _hism_crowd_05)
      (Components|DestroyComponent _hism_crowd_05))))

(fn SpawnCrowd
  (params (bSpawnNow :type Boolean))
  (bind (_worldlocation _relativelocation) (CallFunction|GetLookatLocation))
  (bind _crowdoptions (Variables|LegacyCrowd|GetCrowdOptions))
  (bind _additionalrotation (Variables|LegacyCrowd|GetAdditionalRotation))
  (bind _lookatrandomness (Variables|LegacyCrowd|GetLookAtRandomness))
  (bind _crowdsize (Variables|LegacyCrowd|GetCrowdSize))
  (bind _positionrandomness (Variables|LegacyCrowd|GetPositionRandomness))
  (bind _scalevariance (Variables|LegacyCrowd|GetScaleVariance))
  (bind _sparkscrowdspawner (Variables|Default|GetSparksCrowdSpawner))
  (bind _crowdarea (Variables|LegacyCrowd|GetCrowdArea))
  (Networking|CanExecuteCosmeticEvents
    (:True
      (if bSpawnNow
        (Class|SparksCrowdSpawner|SpawnCrowdNow _sparkscrowdspawner _crowdarea _crowdsize _positionrandomness _scalevariance _relativelocation _lookatrandomness _additionalrotation _crowdoptions)
        (else
          (Class|SparksCrowdSpawner|SpawnCrowd _sparkscrowdspawner _crowdarea _crowdsize _positionrandomness _scalevariance _relativelocation _lookatrandomness _additionalrotation (Variables|LegacyCrowd|GetBatchSize) _crowdoptions))))
    (:False
      _))
  (return))

(fn ConstructionScript
  (bind _crowdarea (Variables|LegacyCrowd|GetCrowdArea))
  (bind _z (.z _crowdarea))
  (bind _returnvalue (Math|Float|Max(Float) _z 5e+01))
  (bind _y (.y _crowdarea))
  (bind _returnvalue_2 (Math|Float|Max(Float) _y 5e+01))
  (bind _x (.x _crowdarea))
  (bind _returnvalue_3 (Math|Float|Max(Float) _x 5e+01))
  (bind _returnvalue_4 (Math|Vector|MakeVector _returnvalue_3 _returnvalue_2 _returnvalue))
  ;[1 1 0] This is for editor use ONLY
  (if (Editor|InEditor)
    (if (TouchPlatform|IsMobileGame)
      (Variables|/SparksMaps/Crowd/SparkCrowd_PCGParams.SparkCrowd_PCGParams_C|SetCrowd%20Population (Variables|/SparksMaps/Crowd/SparkCrowd_PCGParams.SparkCrowd_PCGParams_C|GetMobile%20Crowd%20Population (Variables|Default|GetSparkCrowd_PCGParams)) (Variables|Default|GetSparkCrowd_PCGParams))
      (PCG|GenerateLocal (Variables|Default|GetPCG) true)
      (else
        _))
    (if (Variables|Crowd|GetUseSplineCrowd)
      (CallFunction|SetupSplineVolume)
      (else
        (Utilities|FlowControl|Sequence
          (:then_0
            ;[0 0.7292 0.3064] Place the editor box where the crowd will be
            (Components|Box|SetBoxExtent (Variables|LegacyCrowd|GetEDITOR_Box) (* _returnvalue_4 "0.500000,0.500000,1.000000") false)
            (Transformation|AddRelativeLocation (Variables|LegacyCrowd|GetEDITOR_Box) (Math|Vector|MakeVector 0.0 0.0 (/ (.z _returnvalue_4) 2.0)) false true))
          (:then_1
            ;[0 0.7292 0.3064] Place the Look at sphere
            (Components|Box|SetBoxExtent (Variables|Look|GetEDITOR_LookAtArea) (Math|Vector|MakeVector (Math|Float|Max(Float) (Variables|LegacyCrowd|GetLookAtRandomness) 5e+01) (Math|Float|Max(Float) (Variables|LegacyCrowd|GetLookAtRandomness) 5e+01) 5e+01) false)
            (Transformation|SetWorldLocation (Variables|Look|GetEDITOR_LookAtArea) (CallFunction|GetLookatLocation) false true))
          (:then_2))))))
