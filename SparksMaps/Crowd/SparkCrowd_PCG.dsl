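; PCG graph that scatters instanced, shadowless crowd character meshes inside tagged splines in Sparks maps, tuned by SparkCrowd_PCGParams component properties.
; Interior samples are thinned by mask splines, jittered, ray-projected onto surfaces, culled by population noise and self-pruning, then scaled and rotated.
; Points pass a distance-to-fixed-stage-point step and noise before CrowdMembers supplies meshes.
; A handle-override branch is marked not implemented; some alternate spawners are unconnected.
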(graph SparkCrowd_PCG
  (params (CrowdMembers :type (Array (Object /Script/Engine.StaticMesh)) :default [/SparksMaps/Crowd/Meshes/SM_Sparks_F_BunnyWolf_Crowd.SM_Sparks_F_BunnyWolf_Crowd /SparksMaps/Crowd/Meshes/SM_Sparks_F_CandySummer_Crowd.SM_Sparks_F_CandySummer_Crowd /SparksMaps/Crowd/Meshes/SM_Sparks_F_HeadhunterStar_Crowd.SM_Sparks_F_HeadhunterStar_Crowd /SparksMaps/Crowd/Meshes/SM_Sparks_F_PinkJet_Crowd.SM_Sparks_F_PinkJet_Crowd /SparksMaps/Crowd/Meshes/SM_Sparks_F_PunkDevil_Crowd.SM_Sparks_F_PunkDevil_Crowd /SparksMaps/Crowd/Meshes/SM_Sparks_F_SharpFang_Crowd.SM_Sparks_F_SharpFang_Crowd /SparksMaps/Crowd/Meshes/SM_Sparks_M_DemonHunter_Crowd.SM_Sparks_M_DemonHunter_Crowd /SparksMaps/Crowd/Meshes/SM_Sparks_M_RedShirt_Crowd.SM_Sparks_M_RedShirt_Crowd /SparksMaps/Crowd/Meshes/SM_Sparks_M_SpyTech_Crowd.SM_Sparks_M_SpyTech_Crowd]))
  (bind GetSplineData_3 (Get%20Spline%20Data :ActorSelector (struct :ActorFilter Parent :bIncludeChildren true :ActorSelectionTag SparkCrowdPCG :bSelectMultiple true) :bComponentsMustOverlapSelf false)) ; Searches for Splines tagged with "SparkCrowdPCG"
  (bind GetActorProperty_16 (Get%20Actor%20Property :ActorSelector (struct :ActorFilter Parent) :bProcessAllComponents false :bSanitizeOutputAttributeName false :bSelectComponent true :ComponentClass /SparksMaps/Crowd/SparkCrowd_PCGParams.SparkCrowd_PCGParams_C :OutputAttributeName None :PropertyName Trace%20Length))
  (bind CreateAttribute_12 (Create%20Constant :AttributeTypes (struct :Type Float :DoubleValue -1.0)))
  (bind GetActorProperty_14 (Get%20Actor%20Property :ActorSelector (struct :ActorFilter Parent) :bProcessAllComponents false :bSanitizeOutputAttributeName false :bSelectComponent true :ComponentClass /SparksMaps/Crowd/SparkCrowd_PCGParams.SparkCrowd_PCGParams_C :OutputAttributeName None :PropertyName Trace%20Height))
  (bind MakeVectorAttribute_10 (Make%20Vector%20Attribute :OutputType Vector))
  (bind WorldRayHitQuery_34 (World%20Ray%20Hit%20Query :QueryParams (struct :bOverrideDefaultParams true :RayOrigin (0.0 0.0 5e+02) :bIgnorePCGHits true :bTraceComplex true)))
  (bind AttributeNoise_1 (Attribute%20Noise)) ; Population noise
  ; Get Spline
  (bind SplineSampler_4 (Spline%20Sampler :SamplerParams (struct :Dimension OnInterior :InteriorSampleSpacing 5e+01 :InteriorBorderSampleSpacing 5e+01 :InteriorOrientation FollowCurvature :InteriorDensityFalloffCurve (struct :EditorCurveData (struct :Keys [(struct :Value 1.0) (struct :Time 1.0)])) :bUnbounded true))) ; Get points inside Spline
  ; Population
  (bind CreateAttribute_14 (Create%20Constant :AttributeTypes (struct :Type Float :FloatValue 1.0 :DoubleValue -1.0)))
  (bind GetActorProperty_12 (Get%20Actor%20Property :ActorSelector (struct :ActorFilter Parent) :bProcessAllComponents false :bSanitizeOutputAttributeName false :bSelectComponent true :ComponentClass /SparksMaps/Crowd/SparkCrowd_PCGParams.SparkCrowd_PCGParams_C :OutputAttributeName None :PropertyName Crowd%20Population))
  (bind AttributeMathsOp_16 (Attribute%20Maths%20Op :InputSource1 @LastCreated :InputSource2 @LastCreated :InputSource3 @LastCreated :Operation Subtract :OutputTarget @SourceName))
  (bind AttributeNoise_14 (Attribute%20Noise :NoiseMin -0.1)) ; Population noise
  (bind DensityFilter_8 (Density%20Filter :bInvertFilter true :LowerBound 0.0 :UpperBound 0.1))
  ; Pos
  (bind CreateAttribute_5 (Create%20Constant :AttributeTypes (struct :Type Float :FloatValue -1.0 :DoubleValue -1.0)))
  (bind GetActorProperty_3 (Get%20Actor%20Property :ActorSelector (struct :ActorFilter Parent) :bProcessAllComponents false :bSanitizeOutputAttributeName false :bSelectComponent true :ComponentClass /SparksMaps/Crowd/SparkCrowd_PCGParams.SparkCrowd_PCGParams_C :OutputAttributeName None :PropertyName Position%20Noise))
  (bind AttributeMathsOp_4 (Attribute%20Maths%20Op :InputSource1 @LastCreated :InputSource2 @LastCreated :InputSource3 @LastCreated :Operation Multiply :OutputTarget @SourceName))
  (bind AttributeNoise_8 (Attribute%20Noise :InputSource $Position.XY :Mode Add :NoiseMax 5e+01 :NoiseMin -5e+01)) ; Pos Noise
  ; Mask Handle
  (bind GetActorProperty_0 (Get%20Actor%20Property :bProcessAllComponents false :bSanitizeOutputAttributeName false :bSelectComponent true :ComponentClass /SparksMaps/Crowd/SparkCrowd_PCGParams.SparkCrowd_PCGParams_C :OutputAttributeName None :PropertyName Mask%20Tag))
  (bind GetSplineData_1 (Get%20Spline%20Data :ActorSelector (struct :ActorFilter AllWorldActors :bIncludeChildren true :ActorSelectionTag SparkCrowdHandle :bSelectMultiple true) :bComponentsMustOverlapSelf false)) ; Searches for Splines tagged with "SparkCrowdHandle"
  (bind SplineSampler_12 (Spline%20Sampler :SamplerParams (struct :Dimension OnInterior :InteriorSampleSpacing 75.0 :InteriorBorderSampleSpacing 5e+01 :InteriorDensityFalloffCurve (struct :EditorCurveData (struct :Keys [(struct :Value 1.0) (struct :Time 1.0 :Value 1.0)])) :bUnbounded true))) ; Get points inside Spline
  (bind Difference_2 (Difference :DensityFunction Binary)) ; Remove Cross sections between handles and Volume
  (bind ToPoint_8 (To%20Point))
  ; Density
  (bind GetActorProperty_7 (Get%20Actor%20Property :ActorSelector (struct :ActorFilter Parent) :bProcessAllComponents false :bSanitizeOutputAttributeName false :bSelectComponent true :ComponentClass /SparksMaps/Crowd/SparkCrowd_PCGParams.SparkCrowd_PCGParams_C :OutputAttributeName None :PropertyName Crowd%20Density))
  (bind ExecuteBlueprint_4 (Execute%20Blueprint :BlueprintElementInstance (class:/Game/ArtTools/Procedural/PCGMoss/FN_TransformByDensity.FN_TransformByDensity_C :Transform (struct :Translation (-5e+02 -5e+02 0.0))) :BlueprintElementType /Game/ArtTools/Procedural/PCGMoss/FN_TransformByDensity.FN_TransformByDensity_C))
  (bind TransformPoints_5 (Transform%20Points :bAbsoluteRotation true))
  ; Surface Hit
  (bind Projection_44 (Projection :ProjectionParams (struct :bProjectRotations false))) ; Trace Surface
  (bind ToPoint_46 (To%20Point))
  ; Spawn WPO
  (bind SpawnActor_0 (Spawn%20Actor :AttachOptions Attached :bAllowTemplateActorEditing true :Option CollapseActors :TemplateActorClass /Game/ArtTools/characterWPO/BP_character_WPO.BP_character_WPO_C)) ; Can also spawn BPs
  (bind StaticMeshSpawner_5 (Static%20Mesh%20Spawner :bApplyMeshBoundsToPoints false :MeshSelectorParameters (class:/Script/PCG.PCGMeshSelectorByAttribute :AttributeName MyMesh :TemplateDescriptor (struct :ComponentClass /Script/Engine.HierarchicalInstancedStaticMeshComponent :Mobility Stationary)) :MeshSelectorType /Script/PCG.PCGMeshSelectorByAttribute)) ; Place Single Custom WPO
  (bind StaticMeshSpawner_20 (Static%20Mesh%20Spawner :bApplyMeshBoundsToPoints false :MeshSelectorParameters (class:/Script/PCG.PCGMeshSelectorWeighted :MeshEntries [(struct :Descriptor (struct :StaticMesh /SparksMaps/Crowd/Meshes/SM_Sparks_F_BunnyWolf_Crowd.SM_Sparks_F_BunnyWolf_Crowd :ComponentClass /Script/Engine.HierarchicalInstancedStaticMeshComponent :Mobility Stationary :BodyInstance (struct :VelocitySolverIterationCount 1 :CollisionEnabled NoCollision :CollisionProfileName NoCollision :CollisionResponses (struct :ResponseArray [(struct :Channel Visibility :Response ECR_Ignore) (struct :Channel Camera :Response ECR_Ignore)]) :bAutoWeld true) :bCastShadow false :bCastDynamicShadow false :bCastStaticShadow false :bCastContactShadow false)) (struct :Descriptor (struct :StaticMesh /SparksMaps/Crowd/Meshes/SM_Sparks_F_PunkDevil_Crowd.SM_Sparks_F_PunkDevil_Crowd :ComponentClass /Script/Engine.HierarchicalInstancedStaticMeshComponent :Mobility Movable :BodyInstance (struct :VelocitySolverIterationCount 1 :ObjectType ECC_WorldDynamic :CollisionEnabled QueryOnly :CollisionProfileName BlockAllDynamic :CollisionResponses (struct :ResponseArray [(struct :Channel FCC_Trace_WeaponCapsule) (struct :Channel FCC_Object_PawnAI) (struct :Channel FCC_Trace_WeaponMulti) (struct :Channel FCC_Trace_Weapon) (struct :Channel FCC_Object_Projectile) (struct :Channel FCC_Object_ProjectileIgnoreAI) (struct :Channel FCC_Object_ProjectileMesh) (struct :Channel FCC_Trace_WeaponPawnAndWorldMulti)]) :bAutoWeld true) :bCastShadow false :bCastDynamicShadow false :bCastStaticShadow false :bCastContactShadow false :bGenerateOverlapEvents true)) (struct :Descriptor (struct :StaticMesh /SparksMaps/Crowd/Meshes/SM_Sparks_F_CandySummer_Crowd.SM_Sparks_F_CandySummer_Crowd :ComponentClass /Script/Engine.HierarchicalInstancedStaticMeshComponent :Mobility Movable :BodyInstance (struct :VelocitySolverIterationCount 1 :ObjectType ECC_WorldDynamic :CollisionEnabled QueryOnly :CollisionProfileName BlockAllDynamic :CollisionResponses (struct :ResponseArray [(struct :Channel FCC_Trace_WeaponCapsule) (struct :Channel FCC_Object_PawnAI) (struct :Channel FCC_Trace_WeaponMulti) (struct :Channel FCC_Trace_Weapon) (struct :Channel FCC_Object_Projectile) (struct :Channel FCC_Object_ProjectileIgnoreAI) (struct :Channel FCC_Object_ProjectileMesh) (struct :Channel FCC_Trace_WeaponPawnAndWorldMulti)]) :bAutoWeld true) :bCastShadow false :bCastDynamicShadow false :bCastStaticShadow false :bCastContactShadow false :bGenerateOverlapEvents true)) (struct :Descriptor (struct :StaticMesh /SparksMaps/Crowd/Meshes/SM_Sparks_F_HeadhunterStar_Crowd.SM_Sparks_F_HeadhunterStar_Crowd :ComponentClass /Script/Engine.HierarchicalInstancedStaticMeshComponent :Mobility Stationary :BodyInstance (struct :VelocitySolverIterationCount 1 :CollisionEnabled NoCollision :CollisionProfileName NoCollision :CollisionResponses (struct :ResponseArray [(struct :Channel Visibility :Response ECR_Ignore) (struct :Channel Camera :Response ECR_Ignore)]) :bAutoWeld true) :bCastShadow false :bCastDynamicShadow false :bCastStaticShadow false :bCastContactShadow false)) (struct :Descriptor (struct :StaticMesh /SparksMaps/Crowd/Meshes/SM_Sparks_F_PinkJet_Crowd.SM_Sparks_F_PinkJet_Crowd :ComponentClass /Script/Engine.HierarchicalInstancedStaticMeshComponent :Mobility Stationary :BodyInstance (struct :VelocitySolverIterationCount 1 :CollisionEnabled NoCollision :CollisionProfileName NoCollision :CollisionResponses (struct :ResponseArray [(struct :Channel Visibility :Response ECR_Ignore) (struct :Channel Camera :Response ECR_Ignore)]) :bAutoWeld true) :bCastShadow false :bCastDynamicShadow false :bCastStaticShadow false :bCastContactShadow false)) (struct :Descriptor (struct :StaticMesh /SparksMaps/Crowd/Meshes/SM_Sparks_F_SharpFang_Crowd.SM_Sparks_F_SharpFang_Crowd :ComponentClass /Script/Engine.HierarchicalInstancedStaticMeshComponent :Mobility Stationary :BodyInstance (struct :VelocitySolverIterationCount 1 :CollisionEnabled NoCollision :CollisionProfileName NoCollision :CollisionResponses (struct :ResponseArray [(struct :Channel Visibility :Response ECR_Ignore) (struct :Channel Camera :Response ECR_Ignore)]) :bAutoWeld true) :bCastShadow false :bCastDynamicShadow false :bCastStaticShadow false :bCastContactShadow false)) (struct :Descriptor (struct :StaticMesh /SparksMaps/Crowd/Meshes/SM_Sparks_M_DemonHunter_Crowd.SM_Sparks_M_DemonHunter_Crowd :ComponentClass /Script/Engine.HierarchicalInstancedStaticMeshComponent :Mobility Stationary :BodyInstance (struct :VelocitySolverIterationCount 1 :CollisionEnabled NoCollision :CollisionProfileName NoCollision :CollisionResponses (struct :ResponseArray [(struct :Channel Visibility :Response ECR_Ignore) (struct :Channel Camera :Response ECR_Ignore)]) :bAutoWeld true) :bCastShadow false :bCastDynamicShadow false :bCastStaticShadow false :bCastContactShadow false)) (struct :Descriptor (struct :StaticMesh /SparksMaps/Crowd/Meshes/SM_Sparks_M_RedShirt_Crowd.SM_Sparks_M_RedShirt_Crowd :ComponentClass /Script/Engine.HierarchicalInstancedStaticMeshComponent :Mobility Stationary :BodyInstance (struct :VelocitySolverIterationCount 1 :CollisionEnabled NoCollision :CollisionProfileName NoCollision :CollisionResponses (struct :ResponseArray [(struct :Channel Visibility :Response ECR_Ignore) (struct :Channel Camera :Response ECR_Ignore)]) :bAutoWeld true) :bCastShadow false :bCastDynamicShadow false :bCastStaticShadow false :bCastContactShadow false)) (struct :Descriptor (struct :StaticMesh /SparksMaps/Crowd/Meshes/SM_Sparks_M_SpyTech_Crowd.SM_Sparks_M_SpyTech_Crowd :ComponentClass /Script/Engine.HierarchicalInstancedStaticMeshComponent :Mobility Movable :BodyInstance (struct :VelocitySolverIterationCount 1 :ObjectType ECC_WorldDynamic :CollisionEnabled QueryOnly :CollisionProfileName BlockAllDynamic :CollisionResponses (struct :ResponseArray [(struct :Channel FCC_Trace_WeaponCapsule) (struct :Channel FCC_Object_PawnAI) (struct :Channel FCC_Trace_WeaponMulti) (struct :Channel FCC_Trace_Weapon) (struct :Channel FCC_Object_Projectile) (struct :Channel FCC_Object_ProjectileIgnoreAI) (struct :Channel FCC_Object_ProjectileMesh) (struct :Channel FCC_Trace_WeaponPawnAndWorldMulti)]) :bAutoWeld true) :bCastShadow false :bCastDynamicShadow false :bCastStaticShadow false :bCastContactShadow false :bGenerateOverlapEvents true)) (struct :Descriptor (struct :StaticMesh /SparksMaps/Crowd/Meshes/SM_Sparks_M_RobotTrouble_Crowd.SM_Sparks_M_RobotTrouble_Crowd :ComponentClass /Script/Engine.HierarchicalInstancedStaticMeshComponent :Mobility Movable :BodyInstance (struct :VelocitySolverIterationCount 1 :ObjectType ECC_WorldDynamic :CollisionEnabled QueryOnly :CollisionProfileName BlockAllDynamic :CollisionResponses (struct :ResponseArray [(struct :Channel FCC_Trace_WeaponCapsule) (struct :Channel FCC_Object_PawnAI) (struct :Channel FCC_Trace_WeaponMulti) (struct :Channel FCC_Trace_Weapon) (struct :Channel FCC_Object_Projectile) (struct :Channel FCC_Object_ProjectileIgnoreAI) (struct :Channel FCC_Object_ProjectileMesh) (struct :Channel FCC_Trace_WeaponPawnAndWorldMulti)]) :bAutoWeld true) :bCastShadow false :bCastDynamicShadow false :bCastStaticShadow false :bCastContactShadow false :bGenerateOverlapEvents true))]))) ; Place WPO
  ; Scale
  (bind CreateAttribute_24 (Create%20Constant :AttributeTypes (struct :Type Float :FloatValue 0.5 :DoubleValue -1.0)))
  (bind GetActorProperty_22 (Get%20Actor%20Property :ActorSelector (struct :ActorFilter Parent) :bProcessAllComponents false :bSanitizeOutputAttributeName false :bSelectComponent true :ComponentClass /SparksMaps/Crowd/SparkCrowd_PCGParams.SparkCrowd_PCGParams_C :OutputAttributeName None :PropertyName Scale%20Noise))
  (bind AttributeMathsOp_25 (Attribute%20Maths%20Op :InputSource1 @LastCreated :InputSource2 @LastCreated :InputSource3 @LastCreated :Operation Multiply :OutputTarget @SourceName))
  (bind CreateAttribute_27 (Create%20Constant :AttributeTypes (struct :Type Float :FloatValue 1.0 :DoubleValue -1.0)))
  (bind AttributeMathsOp_29 (Attribute%20Maths%20Op :InputSource1 @LastCreated :InputSource2 @LastCreated :InputSource3 @LastCreated :Operation Subtract :OutputTarget @SourceName))
  (bind AttributeMathsOp_30 (Attribute%20Maths%20Op :InputSource1 @LastCreated :InputSource2 @LastCreated :InputSource3 @LastCreated :OutputTarget @SourceName))
  (bind AttributeNoise_2 (Attribute%20Noise :InputSource $Scale.Z :NoiseMax 1.1 :NoiseMin 0.9)) ; Scale Noise
  ; Remove Overlapping Meshes
  (bind BoundsModifier_2 (Bounds%20Modifier))
  (bind SelfPruning_3 (Self%20Pruning :Parameters (struct :PruningType AllEqual :ComparisonSource $Extents)))
  ; Base Rotation
  (bind ExecuteBlueprint_20 (Execute%20Blueprint :BlueprintElementInstance (class:/PCG/BP_Elements/Deprecated/RebuildPointRotationFromNormal.RebuildPointRotationFromNormal_C) :BlueprintElementType /PCG/BP_Elements/Deprecated/RebuildPointRotationFromNormal.RebuildPointRotationFromNormal_C)) ; Set Rotation to World Zero
  (bind GetActorProperty_11 (Get%20Actor%20Property :ActorSelector (struct :ActorFilter Parent) :bProcessAllComponents false :bSanitizeOutputAttributeName false :bSelectComponent true :ComponentClass /SparksMaps/Crowd/SparkCrowd_PCGParams.SparkCrowd_PCGParams_C :OutputAttributeName None :PropertyName Additional%20Rotation))
  (bind BreakTransformAttribute_13 (Break%20Transform%20Attribute :InputSource $Transform :OutputTarget None)) ; Add Custom Rotation
  (bind MakeTransformAttribute_15 (Make%20Transform%20Attribute :InputSource1 @LastCreated :InputSource2 @LastCreated :InputSource3 @LastCreated :OutputTarget $Transform))
  ;[1 0.005208 0.005208] (Not Implemented) - Modifying Parameters with handles
  (bind GetActorProperty_17 (Get%20Actor%20Property :ActorSelector (struct :ActorFilter AllWorldActors :bMustOverlapSelf true :ActorSelectionTag SparkCrowdHandle :bSelectMultiple true) :bProcessAllComponents false :bSanitizeOutputAttributeName false :bSelectComponent true :ComponentClass /SparksMaps/Crowd/SparkCrowd_PCGParams.SparkCrowd_PCGParams_C :OutputAttributeName None :PropertyName Crowd%20Population))
  (bind CreateAttribute_29 (Create%20Constant :AttributeTypes (struct :Type Float :FloatValue 1.0 :DoubleValue -1.0)))
  (bind AttributeMathsOp_31 (Attribute%20Maths%20Op :InputSource1 @LastCreated :InputSource2 @LastCreated :InputSource3 @LastCreated :Operation Subtract :OutputTarget @SourceName))
  (bind GetSplineData_54 (Get%20Spline%20Data :ActorSelector (struct :ActorFilter AllWorldActors :bMustOverlapSelf true :bIncludeChildren true :ActorSelectionTag SparkCrowdHandle :bSelectMultiple true) :bComponentsMustOverlapSelf false)) ; Searches for Splines tagged with "SparkCrowdHandle"
  (bind SplineSampler_59 (Spline%20Sampler :SamplerParams (struct :Dimension OnInterior :InteriorSampleSpacing 75.0 :InteriorBorderSampleSpacing 5e+01 :InteriorOrientation FollowCurvature :InteriorDensityFalloffCurve (struct :EditorCurveData (struct :Keys [(struct :Value 1.0) (struct :Time 1.0)])) :bUnbounded true))) ; Get points inside Spline
  (bind Projection_24 (Projection :ProjectionParams (struct :bProjectRotations false))) ; Trace Surface
  (bind ToPoint_26 (To%20Point))
  (bind AttributeNoise_25 (Attribute%20Noise :NoiseMin -0.1)) ; Population noise
  (bind DensityFilter_22 (Density%20Filter :bInvertFilter true :LowerBound 0.0 :UpperBound 0.1))
  (bind ExecuteBlueprint_14 (Execute%20Blueprint :BlueprintElementInstance (class:/PCG/BP_Elements/Deprecated/RebuildPointRotationFromNormal.RebuildPointRotationFromNormal_C) :BlueprintElementType /PCG/BP_Elements/Deprecated/RebuildPointRotationFromNormal.RebuildPointRotationFromNormal_C)) ; Set Rotation to World Zero
  (bind BreakTransformAttribute_18 (Break%20Transform%20Attribute :InputSource $Transform)) ; Add Custom Rotation
  (bind GetActorProperty_13 (Get%20Actor%20Property :ActorSelector (struct :ActorFilter AllWorldActors :bMustOverlapSelf true :ActorSelectionTag SparkCrowdHandle :bSelectMultiple true) :bProcessAllComponents false :bSanitizeOutputAttributeName false :bSelectComponent true :ComponentClass /SparksMaps/Crowd/SparkCrowd_PCGParams.SparkCrowd_PCGParams_C :OutputAttributeName None :PropertyName Additional%20Rotation))
  (bind MakeTransformAttribute_20 (Make%20Transform%20Attribute :InputSource1 @LastCreated :InputSource2 @LastCreated :InputSource3 @LastCreated :OutputTarget $Transform))
  (bind StaticMeshSpawner_12 (Static%20Mesh%20Spawner :bApplyMeshBoundsToPoints false :MeshSelectorParameters (class:/Script/PCG.PCGMeshSelectorWeighted :MeshEntries [(struct :Descriptor (struct :StaticMesh /SparksMaps/Crowd/Meshes/Version_2/SM_Sparks_CID_005_M_Crowd.SM_Sparks_CID_005_M_Crowd :ComponentClass /Script/Engine.HierarchicalInstancedStaticMeshComponent :Mobility Stationary :BodyInstance (struct :VelocitySolverIterationCount 1 :CollisionEnabled NoCollision :CollisionProfileName NoCollision :CollisionResponses (struct :ResponseArray [(struct :Channel Visibility :Response ECR_Ignore) (struct :Channel Camera :Response ECR_Ignore)]) :bAutoWeld true))) (struct :Descriptor (struct :StaticMesh /SparksMaps/Crowd/Meshes/Version_2/SM_Sparks_CID_030_M_Crowd.SM_Sparks_CID_030_M_Crowd :ComponentClass /Script/Engine.HierarchicalInstancedStaticMeshComponent :Mobility Stationary :BodyInstance (struct :VelocitySolverIterationCount 1 :CollisionEnabled NoCollision :CollisionProfileName NoCollision :CollisionResponses (struct :ResponseArray [(struct :Channel Visibility :Response ECR_Ignore) (struct :Channel Camera :Response ECR_Ignore)]) :bAutoWeld true))) (struct :Descriptor (struct :StaticMesh /SparksMaps/Crowd/Meshes/Version_2/SM_Sparks_CID_168_F_Crowd.SM_Sparks_CID_168_F_Crowd :ComponentClass /Script/Engine.HierarchicalInstancedStaticMeshComponent :Mobility Stationary :BodyInstance (struct :VelocitySolverIterationCount 1 :CollisionEnabled NoCollision :CollisionProfileName NoCollision :CollisionResponses (struct :ResponseArray [(struct :Channel Visibility :Response ECR_Ignore) (struct :Channel Camera :Response ECR_Ignore)]) :bAutoWeld true))) (struct :Descriptor (struct :StaticMesh /SparksMaps/Crowd/Meshes/Version_2/SM_Sparks_CID_169_F_Crowd.SM_Sparks_CID_169_F_Crowd :ComponentClass /Script/Engine.HierarchicalInstancedStaticMeshComponent :Mobility Stationary :BodyInstance (struct :VelocitySolverIterationCount 1 :CollisionEnabled NoCollision :CollisionProfileName NoCollision :CollisionResponses (struct :ResponseArray [(struct :Channel Visibility :Response ECR_Ignore) (struct :Channel Camera :Response ECR_Ignore)]) :bAutoWeld true))) (struct :Descriptor (struct :StaticMesh /SparksMaps/Crowd/Meshes/Version_2/SM_Sparks_CID_397_F_Crowd.SM_Sparks_CID_397_F_Crowd :ComponentClass /Script/Engine.HierarchicalInstancedStaticMeshComponent :Mobility Stationary :BodyInstance (struct :VelocitySolverIterationCount 1 :CollisionEnabled NoCollision :CollisionProfileName NoCollision :CollisionResponses (struct :ResponseArray [(struct :Channel Visibility :Response ECR_Ignore) (struct :Channel Camera :Response ECR_Ignore)]) :bAutoWeld true))) (struct :Descriptor (struct :StaticMesh /SparksMaps/Crowd/Meshes/Version_2/SM_Sparks_CID_165_M_Crowd.SM_Sparks_CID_165_M_Crowd :ComponentClass /Script/Engine.HierarchicalInstancedStaticMeshComponent :Mobility Stationary :BodyInstance (struct :VelocitySolverIterationCount 1 :CollisionEnabled NoCollision :CollisionProfileName NoCollision :CollisionResponses (struct :ResponseArray [(struct :Channel Visibility :Response ECR_Ignore) (struct :Channel Camera :Response ECR_Ignore)]) :bAutoWeld true)))]))) ; Place WPO
  ; Try to keep most part of the crowd close to the stage
  (bind CreatePoints_28 (Create%20Points :PointsToCreate [(struct :Transform (struct :Translation (-5899.564324395895 -528.0 488.0)))]))
  (bind GetActorProperty_37 (Get%20Actor%20Property :ActorSelector (struct :ActorFilter Parent) :bProcessAllComponents false :bSanitizeOutputAttributeName false :bSelectComponent true :ComponentClass /SparksMaps/Crowd/SparkCrowd_PCGParams.SparkCrowd_PCGParams_C :OutputAttributeName "Maximum Distance" :PropertyName Maximum%20Distance))
  (bind Distance_29 (Distance :bSetDensity true :MaximumDistance 3692.984834 :SourceShape Center :TargetShape Center))
  ; Spawn Crowd
  (bind AttributeNoise_31 (Attribute%20Noise)) ; Population noise
  (bind MatchAndSetAttributes_15 (Match%20And%20Set%20Attributes :MatchAttribute None :Match%20Data CrowdMembers))
  (bind StaticMeshSpawner_27 (Static%20Mesh%20Spawner :bApplyMeshBoundsToPoints false :MeshSelectorParameters (class:/Script/PCG.PCGMeshSelectorByAttribute :AttributeName CrowdMembers :TemplateDescriptor (struct :ComponentClass /Script/Engine.HierarchicalInstancedStaticMeshComponent :bCastShadow false :bAffectDynamicIndirectLighting false :bAffectDistanceFieldLighting false :bCastDynamicShadow false :bCastStaticShadow false :bCastContactShadow false :bReceivesDecals false :bVisibleInRayTracing false :bIncludeInHLOD false :bForceNavigationObstacle false)) :MeshSelectorType /Script/PCG.PCGMeshSelectorByAttribute)) ; Place WPO - Epic
  (connect GetActorProperty_3 Out AttributeMathsOp_4 InA)
  (connect CreateAttribute_5 Out AttributeMathsOp_4 InB)
  (connect GetActorProperty_12 Out AttributeMathsOp_16 InA)
  (connect CreateAttribute_14 Out AttributeMathsOp_16 InB)
  (connect GetActorProperty_22 Out AttributeMathsOp_25 InA)
  (connect CreateAttribute_24 Out AttributeMathsOp_25 InB)
  (connect CreateAttribute_27 Out AttributeMathsOp_29 InA)
  (connect AttributeMathsOp_25 Out AttributeMathsOp_29 InB)
  (connect CreateAttribute_27 Out AttributeMathsOp_30 InA)
  (connect AttributeMathsOp_25 Out AttributeMathsOp_30 InB)
  (connect GetActorProperty_17 Out AttributeMathsOp_31 InA)
  (connect CreateAttribute_29 Out AttributeMathsOp_31 InB)
  (connect SelfPruning_3 Out AttributeNoise_2 In)
  (connect AttributeMathsOp_30 Out AttributeNoise_2 NoiseMax)
  (connect AttributeMathsOp_29 Out AttributeNoise_2 NoiseMin)
  (connect ToPoint_8 Out AttributeNoise_8 In)
  (connect GetActorProperty_3 Out AttributeNoise_8 NoiseMax)
  (connect AttributeMathsOp_4 Out AttributeNoise_8 NoiseMin)
  (connect ToPoint_46 Out AttributeNoise_14 In)
  (connect GetActorProperty_12 Out AttributeNoise_14 NoiseMax)
  (connect AttributeMathsOp_16 Out AttributeNoise_14 NoiseMin)
  (connect ToPoint_26 Out AttributeNoise_25 In)
  (connect GetActorProperty_17 Out AttributeNoise_25 NoiseMax)
  (connect AttributeMathsOp_31 Out AttributeNoise_25 NoiseMin)
  (connect Distance_29 Out AttributeNoise_31 In)
  (connect DensityFilter_8 Out BoundsModifier_2 In)
  (connect ExecuteBlueprint_20 Out BreakTransformAttribute_13 In)
  (connect ExecuteBlueprint_14 Out BreakTransformAttribute_18 In)
  (connect AttributeNoise_14 Out DensityFilter_8 In)
  (connect AttributeNoise_25 Out DensityFilter_22 In)
  (connect SplineSampler_12 Out Difference_2 Differences)
  (connect TransformPoints_5 Out Difference_2 Source)
  (connect GetActorProperty_37 Out Distance_29 MaximumDistance)
  (connect MakeTransformAttribute_15 Out Distance_29 Source)
  (connect CreatePoints_28 Out Distance_29 Target)
  (connect GetActorProperty_7 Out ExecuteBlueprint_4 Density%20Exp)
  (connect SplineSampler_4 Out ExecuteBlueprint_4 In)
  (connect DensityFilter_22 Out ExecuteBlueprint_14 In)
  (connect AttributeNoise_2 Out ExecuteBlueprint_20 In)
  (connect GetActorProperty_0 Out GetSplineData_1 ActorSelectionTag)
  (connect GetActorProperty_11 Out MakeTransformAttribute_15 Rotation)
  (connect BreakTransformAttribute_13 Scale MakeTransformAttribute_15 Scale)
  (connect BreakTransformAttribute_13 Translation MakeTransformAttribute_15 Translation)
  (connect GetActorProperty_13 Out MakeTransformAttribute_20 Rotation)
  (connect BreakTransformAttribute_18 Scale MakeTransformAttribute_20 Scale)
  (connect BreakTransformAttribute_18 Translation MakeTransformAttribute_20 Translation)
  (connect CreateAttribute_12 Out MakeVectorAttribute_10 X)
  (connect CreateAttribute_12 Out MakeVectorAttribute_10 Y)
  (connect GetActorProperty_14 Out MakeVectorAttribute_10 Z)
  (connect AttributeNoise_31 Out MatchAndSetAttributes_15 In)
  (connect SplineSampler_59 Out Projection_24 In)
  (connect WorldRayHitQuery_34 Out Projection_24 Projection%20Target)
  (connect AttributeNoise_8 Out Projection_44 In)
  (connect WorldRayHitQuery_34 Out Projection_44 Projection%20Target)
  (connect BoundsModifier_2 Out SelfPruning_3 In)
  (connect GetSplineData_3 Out SplineSampler_4 Spline)
  (connect GetSplineData_1 Out SplineSampler_12 Spline)
  (connect GetSplineData_54 Out SplineSampler_59 Spline)
  (connect MakeTransformAttribute_20 Out StaticMeshSpawner_12 In)
  (connect MatchAndSetAttributes_15 Out StaticMeshSpawner_27 In)
  (connect Difference_2 Out ToPoint_8 In)
  (connect Projection_24 Out ToPoint_26 In)
  (connect Projection_44 Out ToPoint_46 In)
  (connect ExecuteBlueprint_4 Out TransformPoints_5 In)
  (connect GetActorProperty_16 Out WorldRayHitQuery_34 RayLength)
  (connect MakeVectorAttribute_10 Out WorldRayHitQuery_34 RayOrigin))
